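# Niagara system for the Halloween smoke in the SkippingClouds player emote, built from a single Smoke sprite emitter.
# It spawns a low rate of faint, randomly sized puffs in a thin, wide box.
# A curve-decayed downward force, panning curl noise and random drag drift them while they grow, fade in and out, and animate through SubUV frames.
# The exposed User parameters are not referenced by the shown modules.
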NiagaraSystem /BRCosmetics/Animation/Game/MainPlayer/Emotes/SkippingClouds/FX/NS_SkippingClouds_HalloweenSmoke
dependencies {
 usedDataInterfaces [&NiagaraDataInterfaceCurve, &NiagaraDataInterfaceSkeletalMesh]
 usedDynamicInputs []
 usedModules [
  &self:SystemSpawnScript.SystemScriptSource.SystemScriptGraph.NiagaraNodeAssignment_0.SetVariables_#1
  &SystemState
 ]
}
emitters {Smoke {
  emitterUpdateScript [
   &EmitterState(
    Inactive Response=::NewEnumerator1
    Life Cycle Mode=::NewEnumerator1
    Loop Behavior=::NewEnumerator1
    Loop Delay=1
    Loop Duration Mode=::NewEnumerator1
    UseLoopDelay=true
   )
   &SpawnRate(SpawnRate=5, Use Spawn Probability=false)
  ]
  particleSpawnScript [
   &InitializeParticle(
    Color Maximum=<rgba 1 1 1 0.15>
    Color Minimum=<rgba 0.806 0.9224 1 0.1>
    Color Mode=::NewEnumerator2
    Lifetime Max=6
    Lifetime Min=3
    Lifetime Mode=::NewEnumerator1
    Sprite Rotation Mode=::NewEnumerator2
    Sprite Size Mode=::NewEnumerator1
    Uniform Sprite Size Max=50
    Uniform Sprite Size Min=30
    UsePositionOffset=false
    Write Material Random=true
   )
   &ShapeLocation(Box Size=<80 15 1>, MaskSpawnGroup=false, Override Seed=false, Shape Primitive=::NewEnumerator2, Surface Only=false, UseOwnerScale=true)
  ]
  particleUpdateScript [
   &ParticleState()
   &CameraOffset(Camera Offset Amount=5)
   &GravityForce(Gravity=&MakeVector(
     X=&RandomRangeFloat(Minimum=-1, Maximum=1)
     Y=&RandomRangeFloat(Minimum=-1, Maximum=1)
     Z=&FloatFromCurve(
      FloatCurve=DataInterface() {propertyValues {Curve {keys [
          {arriveTangent=-1.2674, interpMode=RCIM_Cubic, leaveTangent=-1.2674, tangentMode=RCTM_User, time=0, value=1}
          {interpMode=RCIM_Cubic, tangentMode=RCTM_User, time=1, value=0}
         ]}}}
      Scale Curve=&RandomRangeFloat(Minimum=-20, Maximum=-30)
     )
    ))
   &CurlNoiseForce(Noise Frequency=10, Noise Strength=20, Pan Noise=true, Pan Noise Field=<0 0 1>)
   &Drag(Drag=&RandomRangeFloat(Minimum=1, Maximum=3), Use Linear Drag=true, Use Rotational Drag=false)
   &SolveForcesAndVelocity(Clamp Velocity=true, Limit Acceleration=false, Speed Limit=100)
   &ScaleColor(
    Scale Alpha=&FloatFromCurve(
     FloatCurve=DataInterface() {propertyValues {Curve {keys [
         {arriveTangent=4.3482, interpMode=RCIM_Cubic, leaveTangent=4.3482, tangentMode=RCTM_User, time=0, value=0}
         {interpMode=RCIM_Cubic, tangentMode=RCTM_User, time=0.5, value=1}
         {tangentMode=RCTM_User, time=1, value=0}
        ]}}}
     Scale Curve=0.25
    )
    ScaleA=true
    ScaleRGB=true
   )
   &ScaleSpriteSizeBySpeed(Max Scale Factor=<1.1 1.1>, Velocity Threshold=100)
   &ScaleSpriteSize(Uniform Curve Sprite Scale=DataInterface() {propertyValues {Curve {keys [
        {arriveTangent=0.9769, interpMode=RCIM_Cubic, leaveTangent=0.9769, tangentMode=RCTM_User, time=0, value=0.5}
        {interpMode=RCIM_Cubic, tangentMode=RCTM_User, time=1, value=1}
       ]}}})
   &SubUVAnimation()
  ]
  rendererClasses [&NiagaraSpriteRendererProperties]
  renderers [{rendererClass=&NiagaraSpriteRendererProperties}]
 }}
summary {
 systemName=NS_SkippingClouds_HalloweenSmoke
 emitters [{emitterName=Smoke, rendererClasses [&NiagaraSpriteRendererProperties]}]
 userVariables [
  User.bIsFrontEnd {type=NiagaraBool}
  User.bIsFrontEndPreview {type=NiagaraBool}
  User.bIsInWater {type=NiagaraBool}
  User.FXBrightnessScale {type=NiagaraFloat, defaultValue=1, description=""}
  User.FXScale {type=NiagaraFloat, defaultValue=1, description=""}
  User.NormalizedNotifyProgress {type=NiagaraFloat, defaultValue=0, description=""}
 ]
}
