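# Niagara impact effect for the knife hit in the SkippingClouds emote, built from two burst sprite emitters.
# Chunks throws a few small, randomly tinted bits upward in a cone, slowed by drag, pulled down by gravity and shrinking late in life.
# Impact_Spikes flashes brief elongated SubUV sprites that grow and fade at low alpha, with a time-of-day scalar input on the fade.
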NiagaraSystem /BRCosmetics/Animation/Game/MainPlayer/Emotes/SkippingClouds/FX/NS_SkippingClouds_KnifeImpact
dependencies {
 usedDataInterfaces [&NiagaraDataInterfaceCurve, &NiagaraDataInterfaceVector2DCurve, &NiagaraDataInterfaceColorCurve]
 usedDynamicInputs [&HueShiftLinearColor, &SimulationPosition]
 usedModules [&SystemState]
}
emitters {
 Chunks {
  emitterUpdateScript [
   &EmitterState(Life Cycle Mode=::NewEnumerator1, Loop Behavior=::NewEnumerator1, Loop Duration=3, UseLoopDelay=false)
   &SpawnBurst_Instantaneous(
    Spawn Count=&RandomRangeInt(Minimum=3, Maximum=5)
    Use Spawn Probability=false
    UseLoopCountLimit=false
   )
  ]
  particleSpawnScript [
   &InitializeParticle(
    Color=&ColorFromCurve(
     ColorCurve=DataInterface() {propertyValues {
       AlphaCurve {keys [<at 0 1 RCIM_Cubic>]}
       BlueCurve {keys [<at 0 0.2122 RCIM_Cubic>,<at 0.1383 0.2122>,<at 0.3069 0.1859>,<at 0.5553 0.75>,<at 0.7126 0.1859>,<at 1 0.8802>]}
       GreenCurve {keys [<at 0 0.2325 RCIM_Cubic>,<at 0.1383 0.2325>,<at 0.3069 0.2005>,<at 0.5553 0.8456>,<at 0.7126 0.2122>,<at 1 0.9474>]}
       RedCurve {keys [<at 0 0.25 RCIM_Cubic>,<at 0.1383 0.25>,<at 0.3069 0.2188>,<at 0.5553 1>,<at 0.7126 0.2188>,<at 1 1>]}
      }}
     CurveIndex=&UniformRangedFloat(Minimum=0, Maximum=1)
    )
    Lifetime=&UniformRangedFloat(Minimum=0.3, Maximum=0.5)
    Mass Mode=::NewEnumerator1
    Sprite Rotation Angle=&UniformRangedFloat(Minimum=0, Maximum=360)
    Sprite Rotation Mode=::NewEnumerator0
    Sprite Size=&Vector2DFromFloat(Value=&UniformRangedFloat(Minimum=2, Maximum=3))
    Sprite Size Mode=::NewEnumerator4
    UsePositionOffset=false
    Write Material Random=true
   ) {module=InitializeParticle001}
   &AddVelocityInCone(
    Cone Axis=<0 0 1>
    Module.Use Velocity Falloff On Cone Axis=false
    Velocity Distribution Along Cone Axis=0.75
    Velocity Strength=&UniformRangedFloat(Minimum=125, Maximum=175)
   )
  ]
  particleUpdateScript [
   &ParticleState()
   &Drag(Use Linear Drag=true, Use Rotational Drag=false)
   &GravityForce(Gravity=<0 0 -500>)
   &SolveForcesAndVelocity(Clamp Velocity=false, Limit Acceleration=false)
   &ScaleColor(Scale RGB=<1.2 1.2 1.2>, ScaleA=true, ScaleRGB=true)
   &ScaleSpriteSize(Scale Factor=&Vector2DFromFloat(Value=&FloatFromCurve(FloatCurve=DataInterface() {propertyValues {Curve {keys [
          <at 0.7 0.9701 RCIM_Cubic>
          {arriveTangent=-5.6463, interpMode=RCIM_Cubic, leaveTangent=-5.6463, tangentMode=RCTM_User, time=1.0011, value=0.002}
         ]}}})))
   &SpriteRotationRate(Rotation Rate=&RandomRangeFloat(Minimum=-5, Maximum=5))
  ]
  rendererClasses [&NiagaraSpriteRendererProperties]
  renderers [{rendererClass=&NiagaraSpriteRendererProperties}]
 }
 Impact_Spikes {
  emitterUpdateScript [
   &EmitterState(Life Cycle Mode=::NewEnumerator1, Loop Behavior=::NewEnumerator1, Loop Delay=0.03, Loop Duration=3, UseLoopDelay=false)
   &SpawnBurst_Instantaneous(
    Spawn Count=&RandomRangeInt(Minimum=5, Maximum=7)
    Use Spawn Probability=false
    UseLoopCountLimit=false
   )
  ]
  particleSpawnScript [
   &InitializeParticle(
    Color Maximum=<rgba 0.875 0.7711 0.6352 1>
    Color Minimum=<rgba 0.2604 0.2483 0.2257 0.75>
    Color Mode=::NewEnumerator2
    Lifetime Max=0.2
    Lifetime Min=0.1
    Lifetime Mode=::NewEnumerator1
    Sprite Size Max=<3 20>
    Sprite Size Min=<2 10>
    Sprite Size Mode=::NewEnumerator2
    UsePositionOffset=false
    Write Material Random=true
   )
   &AddVelocity(
    Cone Angle=60
    Cone Axis=<0 0 1>
    Module.UseConeFalloff=false
    Module.UseVelDistribution=false
    Override Seed=false
    Velocity Mode=::NewEnumerator2
    Velocity Speed=10
   ) {module=AddVelocity001}
   &SubUVAnimation(End Frame=3, SubUV Animation Mode:ENiagara_SubUVLookupMode=::NewEnumerator2)
  ]
  particleUpdateScript [
   &ParticleState()
   &Drag(Drag=5, Use Linear Drag=true, Use Rotational Drag=false)
   &SolveForcesAndVelocity(Clamp Velocity=false, Limit Acceleration=false)
   &ScaleSpriteSize(
    Non-Uniform Curve Scale=&Multiply_Vector2D_ByFloat(Float=1.5, Vector2D=<1 1>)
    Non-Uniform Curve Sprite Scale=DataInterface() {propertyValues {
      XCurve {keys [
        {arriveTangent=1.0718, interpMode=RCIM_Cubic, leaveTangent=1.0718, tangentMode=RCTM_User, time=0, value=0.375}
        {arriveTangent=0.2552, interpMode=RCIM_Cubic, leaveTangent=0.2552, tangentMode=RCTM_User, time=1, value=1}
       ]}
      YCurve {keys [{arriveTangent=1.7817, interpMode=RCIM_Cubic, leaveTangent=1.7817, tangentMode=RCTM_User, time=0, value=0.5}, <at 1 1.5 RCIM_Cubic>]}
     }}
    Scale Sprite Size Mode=::NewEnumerator3
   )
   &ScaleColor(
    Scale Alpha=&Multiply_Float(
     A=&FloatFromCurve(
      FloatCurve=DataInterface() {propertyValues {Curve {keys [
          {arriveTangent=17.6896, interpMode=RCIM_Cubic, leaveTangent=17.6896, tangentMode=RCTM_User, time=0, value=0}
          {arriveTangent=-2.5249, interpMode=RCIM_Cubic, leaveTangent=-2.5249, tangentMode=RCTM_User, time=0.1408, value=0.9847}
          {arriveTangent=-0.0405, interpMode=RCIM_Cubic, leaveTangent=-0.0405, tangentMode=RCTM_User, time=1, value=0}
         ]}}}
      Scale Curve=&FNTimeOfDayScalar(Night Scale=1, Morning Scale=1, Evening Scale=1)
     )
     B=0.2
    )
    ScaleA=true
    ScaleRGB=false
   )
  ]
  rendererClasses [&NiagaraSpriteRendererProperties]
  renderers [{rendererClass=&NiagaraSpriteRendererProperties}]
 }
}
summary {
 systemName=NS_SkippingClouds_KnifeImpact
 emitters [
  {emitterName=Impact_Spikes, rendererClasses [&NiagaraSpriteRendererProperties]}
  {emitterName=Chunks, rendererClasses [&NiagaraSpriteRendererProperties]}
 ]
 userVariables [
  User.bIsFrontEnd {type=NiagaraBool}
  User.bIsFrontEndPreview {type=NiagaraBool}
  User.bIsInWater {type=NiagaraBool}
  User.NormalizedNotifyProgress {type=NiagaraFloat, defaultValue=0, description=""}
 ]
}
